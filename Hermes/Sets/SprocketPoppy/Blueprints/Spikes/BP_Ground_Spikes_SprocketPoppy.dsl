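; Ground spike hazard for SprocketPoppy that damages and bounces Fortnite pawns falling fast onto its damage collision mesh.
; On overlap, the authority checks downward velocity, applies environment damage, and launches the pawn upward with a random horizontal push from a vector array.
; A health check beforehand triggers death stinger music through the spike music manager for lethal hits.
; The DamageAudio event spawns a placeholder-path bounce sound only on its Remote branch.
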(event OnComponentBeginOverlap(CustomDamageCollisionMesh)
  (params (OverlappedComponent :type Primitive%20Component%20Object%20Reference) (OtherActor :type Actor%20Object%20Reference) (OtherComp :type Primitive%20Component%20Object%20Reference) (OtherBodyIndex :type Integer) (bFromSweep :type Boolean) (SweepResult :type Hit%20Result%20Structure%20%28by%20ref%29))
  (bind (_x _y) (Math|Vector2D|BreakVector2D (Utilities|Array|Get(aref) (Variables|Default|GetRandomVectors) (Math|Random|RandomIntegerinRange 0 5))))
  ; Spike Damage
  (Networking|SwitchHasAuthority
    (:Authority
      (bind _asfort_pawn (Utilities|Casting|CastToFortPawn OtherActor))
      (Variables|Default|SetCurrentPawn _asfort_pawn)
      (if (< (.z (Transformation|GetVelocity (Variables|Default|GetCurrentPawn))) -5e+02)
        (bind _goingtodie (CallFunction|IsPlayerGoingtoDie?))
        (if _goingtodie
          (Class|BPSpikeMusicManagerSprocketPoppy|PlayDeathStringerMusic (Variables|Default|GetSpikeMusicManagerRef) 0))
        (Environment|ApplyGlobalEnvironmentDamagetoActor (Variables|Default|GetCurrentPawn) (Variables|Default|GetSpikeDamage))
        (Character|LaunchCharacter (Variables|Default|GetCurrentPawn) (Math|Vector|MakeVector _x _y 2e+03) false true)
        (CallFunction|DamageAudio)
        (Development|PrintString "Hit Spike" true true "(R=0.000000,G=0.660000,B=1.000000,A=1.000000)" 5.0)))))

(event Custom|DamageAudio
  ; Custom Spike Client Sounds and FX
  (Networking|SwitchHasAuthority
    (:Remote
      (Audio|SpawnSoundatLocation "/Game/Sounds/Fort_Test_Sounds/World_Objects/Bouncing/DivingBoard_Bounce_Impact_Cue.DivingBoard_Bounce_Impact_Cue" (Transformation|GetActorLocation self)))))

(fn IsPlayerGoingToDie?
  (return (<= (- (FortPawn|GetHealth (Variables|Default|GetCurrentPawn)) (Variables|Default|GetSpikeDamage)) 0)))

(fn ConstructionScript)
