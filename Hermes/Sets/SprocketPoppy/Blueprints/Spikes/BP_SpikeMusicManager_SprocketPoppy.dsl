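; Spike music manager for the SprocketPoppy set that plays a death stinger for a pawn involved with ground spikes.
; PlayDeathStringerMusic stores the initiating pawn, then waits briefly so DBNO state can settle.
; If the pawn is valid and not DBNO, a DoOnce gate calls PlayStinger, followed by a StingerResetDuration delay.
; PlayStinger spawns the DadBro horn-destroyed cue at the passed spike's root location, only on non-authority instances.
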(event Custom|PlayDeathStringerMusic
  (params (InitiatingPawnRef :type Fort%20Pawn%20Object%20Reference))
  (bind _initiating_pawn_ref (Variables|Default|GetInitiating%20Pawn%20Ref))
  (Variables|Default|SetInitiating%20Pawn%20Ref InitiatingPawnRef)
  (Utilities|FlowControl|Delay 0.1) ; allow time for dbno state
  (Utilities|IsValid _initiating_pawn_ref
    (:Is%20Valid
      (if (not (Health|IsDBNO _initiating_pawn_ref))
        (Utilities|FlowControl|DoOnce false
          (:Completed
            (CallFunction|PlayStinger 0)
            (Utilities|FlowControl|Delay (Variables|Default|GetStingerResetDuration))))))))

(event Custom|PlayStinger
  (params (SpikeRef :type BP%20Ground%20Spikes%20Sprocket%20Poppy%20Object%20Reference))
  (Networking|SwitchHasAuthority
    (:Remote
      (Audio|SpawnSoundatLocation "/Game/Sounds/Apollo/Events/Dadbro/Music/DadBro_HornDestroyed_Stinger_Cue.DadBro_HornDestroyed_Stinger_Cue" (Transformation|GetWorldLocation (Variables|/Script/Engine.Actor|GetRootComponent SpikeRef))))))

(fn ConstructionScript)
